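; Anim notify state for the Cruisin emote that puts an outline overlay material on the character's skeletal mesh while the notify is active.
; Begin switches on rendering material quality level: level 0 and the default branch apply the same outline instance, while levels 1 and 2 apply none.
; It also caps the overlay's max draw distance.
; End clears the overlay material outright, without restoring any previous one.
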(fn Received_NotifyBegin
  (params (MeshComp :type Skeletal%20Mesh%20Component%20Object%20Reference) (Animation :type Anim%20Sequence%20Base%20Object%20Reference) (TotalDuration :type Float%20%28single-precision%29) (EventReference :type Anim%20Notify%20Event%20Reference%20Structure%20%28by%20ref%29))
  (switch int (Rendering|Material|GetRenderingMaterialQualityLevel)
    (:Default
      (Rendering|Material|SetOverlayMaterial MeshComp "/BRCosmetics/Animation/Game/MainPlayer/Emotes/Cruisin/BP/MI_Cruising_OverlayOutline.MI_Cruising_OverlayOutline" false 0 :!class /Script/Engine.MeshComponent) ; Original Outlines
    )
    (:0
      (Rendering|Material|SetOverlayMaterial MeshComp "/BRCosmetics/Animation/Game/MainPlayer/Emotes/Cruisin/BP/MI_Cruising_OverlayOutline.MI_Cruising_OverlayOutline" false 0 :!class /Script/Engine.MeshComponent) ; Black OverlayOutline
    )
    (:1
      _)
    (:2
      _))
  (Rendering|Material|SetOverlayMaterialMaxDrawDistance MeshComp 1e+04 :!class /Script/Engine.MeshComponent)
  (return false))

(fn Received_NotifyEnd
  (params (MeshComp :type Skeletal%20Mesh%20Component%20Object%20Reference) (Animation :type Anim%20Sequence%20Base%20Object%20Reference) (EventReference :type Anim%20Notify%20Event%20Reference%20Structure%20%28by%20ref%29))
  (Rendering|Material|SetOverlayMaterial MeshComp 0 false 0 :!class /Script/Engine.MeshComponent)
  (return false))
