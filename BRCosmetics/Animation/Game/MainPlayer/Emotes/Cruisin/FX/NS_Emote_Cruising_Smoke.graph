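# Niagara sprite smoke for the Cruisin emote's gadget, with two mirrored side exhaust emitters and a wheel smoke emitter.
# Particles spawn at the FX_Exhaust_Side and FX_Exhaust_Back mesh sockets, using array-selected offsets to place left/right exhausts and four wheel puffs.
# Spawning combines rate and movement distance, plus random bursts for exhausts.
# bIsFrontEnd zeroes spawn rates and softens velocities, while bIsInWater zeroes the per-unit spacing value.
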NiagaraSystem /BRCosmetics/Animation/Game/MainPlayer/Emotes/Cruisin/FX/NS_Emote_Cruising_Smoke
dependencies {
 usedDataInterfaces [
  &NiagaraDataInterfaceSkeletalMesh
  &NiagaraDataInterfaceArrayFloat3
  &NiagaraDataInterfaceCurve
  &NiagaraDataInterfaceArrayFloat
 ]
 usedDynamicInputs []
 usedModules [&SystemState]
}
emitters {
 Exhaust_Smoke_Side {
  emitterUpdateScript [
   &EmitterState(Life Cycle Mode=::NewEnumerator1, Loop Duration=0.2, Recalculate Duration Each Loop=true, UseLoopDelay=false)
   &SpawnRate(
    SpawnRate=&MakeCustomFloatFromBool(
     Bool=$User.bIsFrontEnd
     True Float=&RandomRangeFloat(Minimum=0, Maximum=0)
     False Float=&RandomRangeFloat(Minimum=8, Maximum=10)
    )
    Use Spawn Probability=false
   ) {module=SpawnRate001}
   &SpawnBurst_Instantaneous(
    Spawn Count=&RandomRangeInt(Minimum=0, Maximum=1)
    Spawn Probability=&RandomRangeFloat(Minimum=0, Maximum=0.5)
    Use Spawn Probability=true
    UseLoopCountLimit=false
   )
   &SpawnPerUnit(
    Max Movement Threshold=350
    Module.Use Max Movement Threshold=true
    Module.Use Movement Tolerance=true
    Movement Tolerance=2.5
    Spawn Spacing:NiagaraFloat=&Multiply_Float(
     A=&MakeCustomFloatFromBool(Bool=$User.bIsInWater, True Float=0, False Float=1)
     B=&RandomRangeFloat(Minimum=15, Maximum=25)
    )
    Use Spawn Probability=false
   )
  ]
  particleSpawnScript [
   &FN_SetSpawnGroup(Number of Spawn Groups=1)
   &InitializeParticle(
    Color=<rgba 0.35 0.35 0.35 5>
    Lifetime Max=0.5
    Lifetime Min=0.2
    Lifetime Mode=::NewEnumerator1
    Sprite Rotation Mode=::NewEnumerator2
    Sprite Size Mode=::NewEnumerator1
    Uniform Sprite Size Max=12.5
    Uniform Sprite Size Min=7.5
    UsePositionOffset=false
    Write Material Random=true
   )
   &AddVelocity(Velocity=&MakeCustomVectorFromBool(
     Bool=$User.bIsFrontEnd
     True Vector=&RandomRangeVector(
      Minimum=&SelectVectorFromArray(Direct Array Index=$Particles.SpawnGroup, Vector Selection Array=DataInterface() {propertyValues {FloatData [<0 -30 25>, <0 -30 25>]}})
      Maximum=&SelectVectorFromArray(Direct Array Index=$Particles.SpawnGroup, Vector Selection Array=DataInterface() {propertyValues {FloatData [<50 -50 55>, <50 -50 55>]}})
     )
     False Vector=&RandomRangeVector(
      Minimum=&SelectVectorFromArray(Direct Array Index=$Particles.SpawnGroup, Vector Selection Array=DataInterface() {propertyValues {FloatData [<100 -250 35>, <100 -250 35>]}})
      Maximum=&SelectVectorFromArray(Direct Array Index=$Particles.SpawnGroup, Vector Selection Array=DataInterface() {propertyValues {FloatData [<250 -500 55>, <250 -500 55>]}})
     )
    ))
   &SkeletalMeshLocation(
    Inherit Velocity=false
    MaskSpawnGroup=false
    Mesh Sampling Type=::NewEnumerator5
    Module.Override Randomness=false
    Sampled Position Offset=&SelectVectorFromArray(Direct Array Index=$Particles.SpawnGroup, Vector Selection Array=DataInterface() {propertyValues {FloatData [<45 0 5>, <-45 0 5>]}})
    Sampled Position Offset Coordinate Space:ENiagaraMeshTransforms=::NewEnumerator3
    Skeletal Mesh=DataInterface() {propertyValues {
      DefaultMesh=null
      ExcludeBoneName=root
      OptionalComponentClass=null
      PreviewMesh=/BRCosmetics/Animation/Props/Gadgets_Cruising/Meshes/Cruising_Gadget
      SkinningMode=SkinOnTheFly
      SoftSourceActor=null
      SourceMode=Default
      UvSetIndex=0
      WholeMeshLOD=-1
      bExcludeBone=false
      bReadDeformedGeometry=true
      bRequireCurrentFrameData=true
      ComponentTags []
      FilteredBones []
      FilteredSockets [FX_Exhaust_Side]
      MeshUserParameter {parameter {name=None, typeDefHandle {registeredTypeIndex=1}, varData []}}
      SamplingRegions []
     }}
   )
  ]
  particleUpdateScript [
   /SetVariables(Particles.RandomFloat=&RandomRangeFloat(Minimum=0, Maximum=1)) {isSetParameters=true, module=SetVariables_#1}
   &ParticleState()
   &InheritSourceMovement(
    Applied Position Scale=&Multiply_Vector(
     A=<1 1 0>
     B=&VectorFromFloat(Value=&FloatFromCurve(FloatCurve=DataInterface() {propertyValues {Curve {keys [
           {arriveTangent=9.0725, interpMode=RCIM_Cubic, leaveTangent=9.0725, tangentMode=RCTM_User, time=0, value=1}
           {arriveTangent=-2.5249, interpMode=RCIM_Cubic, leaveTangent=-2.5249, tangentMode=RCTM_User, time=0.1408, value=0.9847}
           {arriveTangent=-0.0405, interpMode=RCIM_Cubic, leaveTangent=-0.0405, tangentMode=RCTM_User, time=1, value=0}
          ]}}}))
    )
    ApplyForce=false
    ApplyPosition=true
    ApplyPositionFalloff=false
   )
   &CameraOffset(Camera Offset Amount=10)
   &Drag(Drag=&RandomRangeFloat(Minimum=7.5, Maximum=12.5), Use Linear Drag=true, Use Rotational Drag=false)
   &GravityForce(Gravity=<0 0 250>)
   &SolveForcesAndVelocity(Clamp Velocity=false, Limit Acceleration=false)
   &ScaleSpriteSize(Scale Factor=&Vector2DFromFloat(Value=&RandomRangeFloat(
      Minimum=&FloatFromCurve(FloatCurve=DataInterface() {propertyValues {Curve {keys [
           {arriveTangent=0.5067, interpMode=RCIM_Cubic, leaveTangent=0.5067, tangentMode=RCTM_User, time=0, value=0.4}
           {arriveTangent=1.25, interpMode=RCIM_Cubic, leaveTangent=1.25, time=0.3, value=0.375}
           {arriveTangent=4.9234, interpMode=RCIM_Cubic, leaveTangent=4.9234, tangentMode=RCTM_User, time=0.4, value=0.9}
           {arriveTangent=-0.75, interpMode=RCIM_Cubic, leaveTangent=-0.75, time=0.6, value=1.35}
           {arriveTangent=-9.268, interpMode=RCIM_Cubic, leaveTangent=-9.268, tangentMode=RCTM_User, time=1, value=0.45}
          ]}}})
      Maximum=&FloatFromCurve(FloatCurve=DataInterface() {propertyValues {Curve {keys [
           {arriveTangent=0.5067, interpMode=RCIM_Cubic, leaveTangent=0.5067, tangentMode=RCTM_User, time=0, value=0.4}
           {arriveTangent=0.9669, interpMode=RCIM_Cubic, leaveTangent=0.9669, time=0.2, value=0.45}
           {arriveTangent=4.9234, interpMode=RCIM_Cubic, leaveTangent=4.9234, tangentMode=RCTM_User, time=0.2796, value=0.6704}
           {arriveTangent=-2.3677, interpMode=RCIM_Cubic, leaveTangent=-2.3677, tangentMode=RCTM_User, time=0.3956, value=1.2136}
           {arriveTangent=-1.9447, interpMode=RCIM_Cubic, leaveTangent=-1.9447, tangentMode=RCTM_User, time=0.9949, value=0.4314}
          ]}}})
     )))
   &ScaleColor(
    Scale Alpha=&FloatFromCurve(FloatCurve=DataInterface() {propertyValues {Curve {keys [
         {interpMode=RCIM_Cubic, tangentMode=RCTM_User, time=0.7, value=1.5}
         {arriveTangent=-2.3042, interpMode=RCIM_Cubic, leaveTangent=-2.3042, tangentMode=RCTM_User, time=1, value=0}
        ]}}})
    Scale RGB=&VectorFromFloat(Value=&FloatFromCurve(FloatCurve=DataInterface() {propertyValues {Curve {keys [
          {arriveTangent=0.3346, interpMode=RCIM_Cubic, leaveTangent=0.3346, tangentMode=RCTM_User, time=0, value=0.15}
          {arriveTangent=-0.0257, interpMode=RCIM_Cubic, leaveTangent=-0.0257, tangentMode=RCTM_User, time=1, value=0.35}
         ]}}}))
    ScaleA=true
    ScaleRGB=true
   )
   &DynamicMaterialParameters(
    Index 0 Param 1=&Add_Float(
     A=$Particles.RandomFloat
     B=&FloatFromCurve(FloatCurve=DataInterface() {propertyValues {Curve {keys [
          {arriveTangent=2.7533, interpMode=RCIM_Cubic, leaveTangent=2.847, tangentMode=RCTM_Break, time=0, value=0}
          {arriveTangent=1.3581, interpMode=RCIM_Cubic, leaveTangent=1.3581, tangentMode=RCTM_User, time=0.3, value=0.55}
          {arriveTangent=0.0529, interpMode=RCIM_Cubic, leaveTangent=0.2021, tangentMode=RCTM_User, time=1, value=1}
         ]}}})
    )
    Index 0 Param 4=&FloatFromCurve(
     FloatCurve=DataInterface() {propertyValues {Curve {keys [
         {interpMode=RCIM_Cubic, tangentMode=RCTM_User, time=0.2, value=1}
         {arriveTangent=-2.3915, interpMode=RCIM_Cubic, leaveTangent=-2.3915, tangentMode=RCTM_User, time=1, value=0}
        ]}}}
     Scale Curve=0
    )
    Module.Param0WriteEnabled=true
    Module.Param1WriteEnabled=false
    Module.Param2WriteEnabled=true
    Module.Param3WriteEnabled=true
   )
  ]
  rendererClasses [&NiagaraSpriteRendererProperties]
  renderers [{rendererClass=&NiagaraSpriteRendererProperties}]
 }
 Exhaust_Smoke_Side001 {
  emitterUpdateScript [
   &EmitterState(Life Cycle Mode=::NewEnumerator1, Loop Duration=0.2, Recalculate Duration Each Loop=true, UseLoopDelay=false)
   &SpawnRate(
    SpawnRate=&MakeCustomFloatFromBool(
     Bool=$User.bIsFrontEnd
     True Float=&RandomRangeFloat(Minimum=0, Maximum=0)
     False Float=&RandomRangeFloat(Minimum=8, Maximum=10)
    )
    Use Spawn Probability=false
   ) {module=SpawnRate001}
   &SpawnBurst_Instantaneous(
    Spawn Count=&RandomRangeInt(Minimum=0, Maximum=1)
    Spawn Probability=&RandomRangeFloat(Minimum=0, Maximum=0.5)
    Use Spawn Probability=true
    UseLoopCountLimit=false
   )
   &SpawnPerUnit(
    Max Movement Threshold=350
    Module.Use Max Movement Threshold=true
    Module.Use Movement Tolerance=true
    Movement Tolerance=2.5
    Spawn Spacing:NiagaraFloat=&Multiply_Float(
     A=&MakeCustomFloatFromBool(Bool=$User.bIsInWater, True Float=0, False Float=1)
     B=&RandomRangeFloat(Minimum=15, Maximum=25)
    )
    Use Spawn Probability=false
   )
  ]
  particleSpawnScript [
   &FN_SetSpawnGroup(Number of Spawn Groups=1)
   &InitializeParticle(
    Color=<rgba 0.35 0.35 0.35 5>
    Lifetime Max=0.5
    Lifetime Min=0.2
    Lifetime Mode=::NewEnumerator1
    Sprite Rotation Mode=::NewEnumerator2
    Sprite Size Mode=::NewEnumerator1
    Uniform Sprite Size Max=12.5
    Uniform Sprite Size Min=7.5
    UsePositionOffset=false
    Write Material Random=true
   )
   &AddVelocity(Velocity=&MakeCustomVectorFromBool(
     Bool=$User.bIsFrontEnd
     True Vector=&RandomRangeVector(
      Minimum=&SelectVectorFromArray(Direct Array Index=$Particles.SpawnGroup, Vector Selection Array=DataInterface() {propertyValues {FloatData [<0 -30 25>, <0 -30 25>]}})
      Maximum=&SelectVectorFromArray(Direct Array Index=$Particles.SpawnGroup, Vector Selection Array=DataInterface() {propertyValues {FloatData [<-50 -50 55>, <-50 -50 55>]}})
     )
     False Vector=&RandomRangeVector(
      Minimum=&SelectVectorFromArray(
       Direct Array Index=$Particles.SpawnGroup
       Vector Selection Array=DataInterface() {propertyValues {FloatData [<-100 -250 35>, <-100 -250 35>]}}
      )
      Maximum=&SelectVectorFromArray(
       Direct Array Index=$Particles.SpawnGroup
       Vector Selection Array=DataInterface() {propertyValues {FloatData [<-250 -500 -55>, <-250 -500 -55>]}}
      )
     )
    ))
   &SkeletalMeshLocation(
    Inherit Velocity=false
    MaskSpawnGroup=false
    Mesh Sampling Type=::NewEnumerator5
    Module.Override Randomness=false
    Sampled Position Offset=&SelectVectorFromArray(Direct Array Index=1, Vector Selection Array=DataInterface() {propertyValues {FloatData [<45 0 5>, <-45 0 5>]}})
    Sampled Position Offset Coordinate Space:ENiagaraMeshTransforms=::NewEnumerator3
    Skeletal Mesh=DataInterface() {propertyValues {
      DefaultMesh=null
      ExcludeBoneName=root
      OptionalComponentClass=null
      PreviewMesh=&Cruising_Gadget
      SkinningMode=SkinOnTheFly
      SoftSourceActor=null
      SourceMode=Default
      UvSetIndex=0
      WholeMeshLOD=-1
      bExcludeBone=false
      bReadDeformedGeometry=true
      bRequireCurrentFrameData=true
      ComponentTags []
      FilteredBones []
      FilteredSockets [FX_Exhaust_Side]
      MeshUserParameter {parameter {name=None, typeDefHandle {registeredTypeIndex=1}, varData []}}
      SamplingRegions []
     }}
   )
  ]
  particleUpdateScript [
   /SetVariables(Particles.RandomFloat=&RandomRangeFloat(Minimum=0, Maximum=1)) {isSetParameters=true, module=SetVariables_#1}
   &ParticleState()
   &InheritSourceMovement(
    Applied Position Scale=&Multiply_Vector(
     A=<1 1 0>
     B=&VectorFromFloat(Value=&FloatFromCurve(FloatCurve=DataInterface() {propertyValues {Curve {keys [
           {arriveTangent=9.0725, interpMode=RCIM_Cubic, leaveTangent=9.0725, tangentMode=RCTM_User, time=0, value=1}
           {arriveTangent=-2.5249, interpMode=RCIM_Cubic, leaveTangent=-2.5249, tangentMode=RCTM_User, time=0.1408, value=0.9847}
           {arriveTangent=-0.0405, interpMode=RCIM_Cubic, leaveTangent=-0.0405, tangentMode=RCTM_User, time=1, value=0}
          ]}}}))
    )
    ApplyForce=false
    ApplyPosition=true
    ApplyPositionFalloff=false
   )
   &CameraOffset(Camera Offset Amount=10)
   &Drag(Drag=&RandomRangeFloat(Minimum=7.5, Maximum=12.5), Use Linear Drag=true, Use Rotational Drag=false)
   &GravityForce(Gravity=<0 0 250>)
   &SolveForcesAndVelocity(Clamp Velocity=false, Limit Acceleration=false)
   &ScaleSpriteSize(Scale Factor=&Vector2DFromFloat(Value=&RandomRangeFloat(
      Minimum=&FloatFromCurve(FloatCurve=DataInterface() {propertyValues {Curve {keys [
           {arriveTangent=0.5067, interpMode=RCIM_Cubic, leaveTangent=0.5067, tangentMode=RCTM_User, time=0, value=0.4}
           {arriveTangent=1.25, interpMode=RCIM_Cubic, leaveTangent=1.25, time=0.3, value=0.375}
           {arriveTangent=4.9234, interpMode=RCIM_Cubic, leaveTangent=4.9234, tangentMode=RCTM_User, time=0.4, value=0.9}
           {arriveTangent=-0.75, interpMode=RCIM_Cubic, leaveTangent=-0.75, time=0.6, value=1.35}
           {arriveTangent=-9.268, interpMode=RCIM_Cubic, leaveTangent=-9.268, tangentMode=RCTM_User, time=1, value=0.45}
          ]}}})
      Maximum=&FloatFromCurve(FloatCurve=DataInterface() {propertyValues {Curve {keys [
           {arriveTangent=0.5067, interpMode=RCIM_Cubic, leaveTangent=0.5067, tangentMode=RCTM_User, time=0, value=0.4}
           {arriveTangent=0.9669, interpMode=RCIM_Cubic, leaveTangent=0.9669, time=0.2, value=0.45}
           {arriveTangent=4.9234, interpMode=RCIM_Cubic, leaveTangent=4.9234, tangentMode=RCTM_User, time=0.2796, value=0.6704}
           {arriveTangent=-2.3677, interpMode=RCIM_Cubic, leaveTangent=-2.3677, tangentMode=RCTM_User, time=0.3956, value=1.2136}
           {arriveTangent=-1.9447, interpMode=RCIM_Cubic, leaveTangent=-1.9447, tangentMode=RCTM_User, time=0.9949, value=0.4314}
          ]}}})
     )))
   &ScaleColor(
    Scale Alpha=&FloatFromCurve(FloatCurve=DataInterface() {propertyValues {Curve {keys [
         {interpMode=RCIM_Cubic, tangentMode=RCTM_User, time=0.7, value=1.5}
         {arriveTangent=-2.3042, interpMode=RCIM_Cubic, leaveTangent=-2.3042, tangentMode=RCTM_User, time=1, value=0}
        ]}}})
    Scale RGB=&VectorFromFloat(Value=&FloatFromCurve(FloatCurve=DataInterface() {propertyValues {Curve {keys [
          {arriveTangent=0.3346, interpMode=RCIM_Cubic, leaveTangent=0.3346, tangentMode=RCTM_User, time=0, value=0.15}
          {arriveTangent=-0.0257, interpMode=RCIM_Cubic, leaveTangent=-0.0257, tangentMode=RCTM_User, time=1, value=0.35}
         ]}}}))
    ScaleA=true
    ScaleRGB=true
   )
   &DynamicMaterialParameters(
    Index 0 Param 1=&Add_Float(
     A=$Particles.RandomFloat
     B=&FloatFromCurve(FloatCurve=DataInterface() {propertyValues {Curve {keys [
          {arriveTangent=2.7533, interpMode=RCIM_Cubic, leaveTangent=2.847, tangentMode=RCTM_Break, time=0, value=0}
          {arriveTangent=1.3581, interpMode=RCIM_Cubic, leaveTangent=1.3581, tangentMode=RCTM_User, time=0.3, value=0.55}
          {arriveTangent=0.0529, interpMode=RCIM_Cubic, leaveTangent=0.2021, tangentMode=RCTM_User, time=1, value=1}
         ]}}})
    )
    Index 0 Param 4=&FloatFromCurve(
     FloatCurve=DataInterface() {propertyValues {Curve {keys [
         {interpMode=RCIM_Cubic, tangentMode=RCTM_User, time=0.2, value=1}
         {arriveTangent=-2.3915, interpMode=RCIM_Cubic, leaveTangent=-2.3915, tangentMode=RCTM_User, time=1, value=0}
        ]}}}
     Scale Curve=0
    )
    Module.Param0WriteEnabled=true
    Module.Param0WriteEnabled1=true
    Module.Param1WriteEnabled=false
    Module.Param1WriteEnabled1=true
    Module.Param2WriteEnabled=true
    Module.Param2WriteEnabled1=true
    Module.Param3WriteEnabled=true
    Module.Param3WriteEnabled1=true
    Write Parameter Index 1=true
   )
  ]
  rendererClasses [&NiagaraSpriteRendererProperties]
  renderers [{rendererClass=&NiagaraSpriteRendererProperties}]
 }
 WheelSmoke_01 {
  emitterUpdateScript [
   &EmitterState()
   &SpawnRate(
    SpawnRate=&MakeCustomFloatFromBool(Bool=$User.bIsFrontEnd, True Float=0, False Float=16)
    Use Spawn Probability=true
   )
   &SpawnPerUnit(
    Max Movement Threshold=750
    Module.Use Max Movement Threshold=true
    Module.Use Movement Tolerance=true
    Movement Tolerance=2
    Spawn Spacing:NiagaraFloat=&Multiply_Float(
     A=&MakeCustomFloatFromBool(Bool=$User.bIsInWater, True Float=0, False Float=1)
     B=&RandomRangeFloat(Minimum=10, Maximum=25)
    )
    Use Spawn Probability=false
   )
  ]
  particleSpawnScript [
   &FN_SetSpawnGroup()
   &InitializeParticle(
    Color=<rgba 0.95 0.8946 0.855 0.2>
    Lifetime=&RandomRangeFloat(Minimum=0.4, Maximum=0.6)
    Sprite Rotation Mode=::NewEnumerator2
    Sprite Size Mode=::NewEnumerator1
    Uniform Sprite Size Max=50
    Uniform Sprite Size Min=30
    UsePositionOffset=false
    Write Material Random=true
   )
   &AddVelocity(Velocity=&MakeCustomVectorFromBool(
     Bool=$User.bIsFrontEnd
     True Vector=&RandomRangeVector(Minimum=<0 -30 25>, Maximum=<0 -50 50>)
     False Vector=&RandomRangeVector(Minimum=<0 -200 25>, Maximum=<0 -400 50>)
    ))
   &SkeletalMeshLocation(
    Inherit Velocity=false
    MaskSpawnGroup=false
    Mesh Sampling Type=::NewEnumerator5
    Module.Override Randomness=false
    Sampled Position Offset=&SelectVectorFromArray(
     Direct Array Index=$Particles.SpawnGroup
     Vector Selection Array=DataInterface() {propertyValues {FloatData [<40 20 0>, <-40 20 0>, <30 120 0>, <-30 120 0>]}}
    )
    Sampled Position Offset Coordinate Space:ENiagaraMeshTransforms=::NewEnumerator3
    Skeletal Mesh=DataInterface() {propertyValues {
      DefaultMesh=null
      ExcludeBoneName=root
      OptionalComponentClass=null
      PreviewMesh=&Cruising_Gadget
      SkinningMode=SkinOnTheFly
      SoftSourceActor=null
      SourceMode=Default
      UvSetIndex=0
      WholeMeshLOD=-1
      bExcludeBone=false
      bReadDeformedGeometry=true
      bRequireCurrentFrameData=true
      ComponentTags []
      FilteredBones []
      FilteredSockets [FX_Exhaust_Back]
      MeshUserParameter {parameter {name=None, typeDefHandle {registeredTypeIndex=1}, varData []}}
      SamplingRegions []
     }}
   )
  ]
  particleUpdateScript [
   /SetVariables(Particles.RandomFloat=&RandomRangeFloat(Minimum=0, Maximum=1)) {isSetParameters=true, module=SetVariables_#1}
   &ParticleState()
   &Drag(Drag=5, Use Linear Drag=true, Use Rotational Drag=false)
   &GravityForce(Gravity=&RandomRangeVector(Minimum=<0 0 220>, Maximum=<0 0 480>))
   &SolveForcesAndVelocity(Clamp Velocity=false, Limit Acceleration=false)
   &ScaleSpriteSize(Scale Factor=&Vector2DFromFloat(Value=&FloatFromCurve(
      FloatCurve=DataInterface() {propertyValues {Curve {keys [
          {arriveTangent=7.9192, interpMode=RCIM_Cubic, leaveTangent=7.9192, tangentMode=RCTM_User, time=0, value=0}
          {arriveTangent=0.9351, interpMode=RCIM_Cubic, leaveTangent=0.9351, tangentMode=RCTM_User, time=0.2, value=0.75}
          {arriveTangent=0.0643, interpMode=RCIM_Cubic, leaveTangent=0.0643, tangentMode=RCTM_User, time=1, value=1.2}
         ]}}}
      Scale Curve=&SelectFloatFromArray(Direct Array Index=$Particles.SpawnGroup, Float Selection Array=DataInterface() {propertyValues {FloatData [2,2,1,1]}})
     )))
   &ScaleColor(
    Scale Alpha=&FloatFromCurve(FloatCurve=DataInterface() {propertyValues {Curve {keys [
         {arriveTangent=-0.7199, interpMode=RCIM_Cubic, leaveTangent=-0.7199, tangentMode=RCTM_User, time=0, value=1}
         {arriveTangent=0.099, interpMode=RCIM_Cubic, leaveTangent=0.099, tangentMode=RCTM_User, time=1, value=0}
        ]}}})
    Scale RGB=&VectorFromFloat(Value=&FloatFromCurve(FloatCurve=DataInterface() {propertyValues {Curve {keys [
          {arriveTangent=0.3346, interpMode=RCIM_Cubic, leaveTangent=0.3346, tangentMode=RCTM_User, time=0, value=0.15}
          {arriveTangent=-0.0257, interpMode=RCIM_Cubic, leaveTangent=-0.0257, tangentMode=RCTM_User, time=1, value=0.35}
         ]}}}))
    ScaleA=true
    ScaleRGB=true
   )
   &DynamicMaterialParameters(
    Index 0 Param 1=&Add_Float(
     A=$Particles.RandomFloat
     B=&FloatFromCurve(FloatCurve=DataInterface() {propertyValues {Curve {keys [
          {arriveTangent=2.7533, interpMode=RCIM_Cubic, leaveTangent=2.847, tangentMode=RCTM_Break, time=0, value=0}
          {arriveTangent=1.3581, interpMode=RCIM_Cubic, leaveTangent=1.3581, tangentMode=RCTM_User, time=0.3, value=0.55}
          {arriveTangent=0.0529, interpMode=RCIM_Cubic, leaveTangent=0.2021, tangentMode=RCTM_User, time=1, value=1}
         ]}}})
    )
    Index 0 Param 4=&FloatFromCurve(FloatCurve=DataInterface() {propertyValues {Curve {keys [
         {interpMode=RCIM_Cubic, tangentMode=RCTM_User, time=0.2, value=1}
         {arriveTangent=-2.3915, interpMode=RCIM_Cubic, leaveTangent=-2.3915, tangentMode=RCTM_User, time=1, value=0}
        ]}}})
    Index 2 Param 1=&FloatFromCurve(FloatCurve=DataInterface() {propertyValues {Curve {keys [
         {arriveTangent=2.9375, interpMode=RCIM_Cubic, leaveTangent=2.9375, tangentMode=RCTM_User, time=0, value=0}
         {leaveTangent=-0.0719, time=1, value=1}
        ]}}})
    Index 2 Param 2=&FloatFromCurve(
     FloatCurve=DataInterface() {propertyValues {Curve {keys [
         {arriveTangent=1.9056, time=0, value=0.2}
         {arriveTangent=0.0259, interpMode=RCIM_Cubic, leaveTangent=0.0259, tangentMode=RCTM_User, time=1, value=1}
        ]}}}
     Scale Curve=&RandomRangeFloat(Minimum=0.5, Maximum=1)
    )
    Index 2 Param 3=&FloatFromCurve(FloatCurve=DataInterface() {propertyValues {Curve {keys [
         {interpMode=RCIM_Cubic, leaveTangent=2.3064, tangentMode=RCTM_Break, time=0, value=1}
         {interpMode=RCIM_Cubic, leaveTangent=11.1233, tangentMode=RCTM_Break, time=1, value=2}
        ]}}})
    Index 2 Param 4=&FloatFromCurve(
     FloatCurve=DataInterface() {propertyValues {Curve {keys [
         {arriveTangent=-2.2264, interpMode=RCIM_Cubic, leaveTangent=-2.2264, tangentMode=RCTM_User, time=0, value=1}
         {interpMode=RCIM_Cubic, leaveTangent=-1.2956, tangentMode=RCTM_Break, time=1, value=0.35}
        ]}}}
     Scale Curve=1.6
    )
    Module.Param0WriteEnabled=true
    Module.Param0WriteEnabled1=true
    Module.Param0WriteEnabled2=true
    Module.Param1WriteEnabled=false
    Module.Param1WriteEnabled1=true
    Module.Param1WriteEnabled2=true
    Module.Param2WriteEnabled=true
    Module.Param2WriteEnabled1=true
    Module.Param2WriteEnabled2=true
    Module.Param3WriteEnabled=true
    Module.Param3WriteEnabled1=true
    Module.Param3WriteEnabled2=true
    Write Parameter Index 1=true
    Write Parameter Index 2=true
   )
   &CameraOffset(Camera Offset Amount=10)
  ]
  rendererClasses [&NiagaraSpriteRendererProperties]
  renderers [{rendererClass=&NiagaraSpriteRendererProperties}]
 }
}
summary {
 systemName=NS_Emote_Cruising_Smoke
 emitters [
  {emitterName=WheelSmoke_01, rendererClasses [&NiagaraSpriteRendererProperties]}
  {emitterName=Exhaust_Smoke_Side, rendererClasses [&NiagaraSpriteRendererProperties]}
  {emitterName=Exhaust_Smoke_Side001, rendererClasses [&NiagaraSpriteRendererProperties]}
 ]
 userVariables [
  User.bIsFrontEnd {type=NiagaraBool}
  User.bIsFrontEndPreview {type=NiagaraBool}
  User.bIsInWater {type=NiagaraBool}
 ]
}
